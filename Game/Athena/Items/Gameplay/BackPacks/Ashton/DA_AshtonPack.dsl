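; Deco tool for the Ashton backpack that acts as an input-to-ability event manager for the pawn carrying it.
; On BeginPlay it binds the pawn's landed event, which sets HasLanded and removes jump and skydive trail effects.
; Primary and secondary fire send gameplay events unless Ashton ability tags are present.
; Primary also requires HasLanded and passes PunchPicker as event magnitude, while secondary clears HasLanded when falling.
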(event EventBeginPlay
  ;[1 1 1] Ability handler. This deco tool acts as an event manager for the abilities running when the item is picked up.
  ;;[0.4711 0.4898 1] EVENT -> BeginPlay:  When you gain this item, bind your land event
  (bind _asfort_pawn (Utilities|Casting|CastToFortPawn (Game|GetInstigator)))
  (EventDispatchers|BindEventtoOnPawnLanded _asfort_pawn (delegate Custom|Landed)))

(event Custom|Landed
  (params (Hit :type Hit%20Result%20Structure%20%28by%20ref%29))
  ;[1 1 1] Ability handler. This deco tool acts as an event manager for the abilities running when the item is picked up.
  ;;[1 1 1] When you land, set your variables or remove other effects
  (Variables|Default|SetHasLanded true)
  (bind _asfort_pawn (Utilities|Casting|CastToFortPawn (Game|GetInstigator)))
  (GameplayEffects|RemoveActiveEffectswithSourceTags (Ability|GetAbilitySystemComponent _asfort_pawn) (Variables|Default|GetGC_JumpTrails))
  (GameplayEffects|RemoveActiveEffectswithSourceTags (Ability|GetAbilitySystemComponent _asfort_pawn) (Variables|Default|GetGC_SkydiveTrails)))

(event DecoTool|EventOnPressTrigger
  (params (FortDecoHelper :type Fort%20Deco%20Helper%20Object%20Reference))
  ;[1 1 1] Ability handler. This deco tool acts as an event manager for the abilities running when the item is picked up.
  ;;[0.3904 0.4224 1] EVENT -> PrimaryFire
  (bind _returnvalue (GameplayTags|HasAnyMatchingGameplayTags(Message) (Game|GetInstigator) (Variables|Default|GetAshtonAbilityTags)))
  (if (and (not _returnvalue) (Variables|Default|GetHasLanded))
    (Math|Float|CompareFloat (Variables|Default|GetPunchPicker) 0.5
      (:>
        (Variables|Default|SetPunchPicker 0.0))
      (:==
        (Variables|Default|SetPunchPicker (+ (Variables|Default|GetPunchPicker) 0.5)))
      (:<
        _))
    (Ability|SendGameplayEventtoActor (Game|GetInstigator) (Variables|Default|GetPrimaryFireEventTag) (Utilities|Struct|MakeGameplayEventData :ContextHandle "()" :InstigatorTags "(GameplayTags=)" :TargetTags "(GameplayTags=)" :EventMagnitude (Variables|Default|GetPunchPicker) :TargetData "()"))
    (Development|PrintString "Primary Fire")))

(event DecoTool|EventOnPressSecondaryFire
  (params (FortDecoHelper :type Fort%20Deco%20Helper%20Object%20Reference))
  ;[1 1 1] Ability handler. This deco tool acts as an event manager for the abilities running when the item is picked up.
  ;;[0.3882 1 0.5623] EVENT -> SecondaryFire
  (bind _returnvalue (GameplayTags|HasAnyMatchingGameplayTags(Message) (Game|GetInstigator) (Variables|Default|GetAshtonAbilityTags)))
  (if (not _returnvalue)
    (bind _asfort_pawn (Utilities|Casting|CastToFortPawn (Game|GetInstigator)))
    (if (AI|Components|NavMovement|IsFalling (Variables|/Script/Engine.Character|GetCharacterMovement _asfort_pawn))
      (if (Variables|Default|GetHasLanded)
        (Variables|Default|SetHasLanded false))
      (else
        _))
    (Ability|SendGameplayEventtoActor (Game|GetInstigator) (Variables|Default|GetSecondaryFireEventTag) (Utilities|Struct|MakeGameplayEventData "(TagName=\"Abilities.Generic.ActiveAbility.Activate\")"))
    (Development|PrintString "Alt Fire")))

(fn ConstructionScript)
