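; Parent gameplay ability for Ashton catches, playing a catch montage and gameplay cue on the player.
; With the checked weapon equipped and no skydiving, it blocks tagged abilities and plays the montage; otherwise it fires only the cue.
; A weapon-switch event or failsafe timer ends it early, and ending unblocks abilities and clears the holster if the montage started.
; Child Blueprints can override HandleMontageCompleted, which by default ends the ability only off-server.
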(event Custom|FailsafeEnd
  ;[1 1 1] If the ability is running for too long kill it.
  (Utilities|FlowControl|DoOnce false
    (:Completed
      (Utilities|Time|ClearandInvalidateTimerbyHandle (Variables|CatchVars|GetTimer_FailsafeEnd))
      (Ability|EndAbility))))

(event Ability|EventActivateAbility
  (bind _returnvalue (Ability|GetAvatarActorfromActorInfo))
  ;[1 1 1] Save Out Player
  (Utilities|IsValid _returnvalue
    (:Is%20Valid
      (bind _asfort_player_pawn (Utilities|Casting|CastToFortPlayerPawn _returnvalue)
        (:then
          (bind _output_get (Variables|CatchVars|SetPlayerPawn _asfort_player_pawn))
          ;[1 1 1] Check if we have the weapon out and we are not skydiving
          (if (and (Utilities|IsA (Variables|/Script/FortniteGame.FortPawn|GetCurrentWeapon (Variables|CatchVars|GetPlayerPawn)) (Variables|CatchVars|GetWeaponToCheckPlayAnim)) (not (Parachute|IsSkydiving _output_get)))
            ;[1 1 1] Hide and Holster weapon
            (Tags|BlockAbilitiesWithTags (Utilities|Casting|CastToFortAbilitySystemComponent (Ability|GetAbilitySystemComponent _output_get)) (Variables|CatchVars|GetAbilityBlockedTags))
            ;[1 1 1] Play Montage and then end ability when finished. Make sure we do once.
            (Ability|Tasks|PlayMontageandWaitforTrigger "None" (Variables|CatchVars|GetMontageToPlay)
              (:then
                (Variables|Default|SetStartedMontage true)
                (Ability|Tasks|WaitGameplayEvent "(TagName=\"Event.Weapons.WeaponSwitch\")"
                  (:then
                    ;[1 1 1] Send GC when we start the montage
                    (Ability|ExecuteGameplayCueOnOwner (Variables|CatchVars|GetT_CatchGC))
                    ;[1 1 1] If the ability is running for too long kill it.
                    (bind _returnvalue_2 (Utilities|Time|SetTimerbyEvent (delegate Custom|FailsafeEnd) (Variables|CatchVars|GetFailsafeDelay)))
                    (Variables|CatchVars|SetTimer_FailsafeEnd _returnvalue_2))
                  (:EventReceived
                    ;[1 1 1] Stop montage if it is playing
                    (Animation|StopAnimMontage (Variables|CatchVars|GetPlayerPawn) (Variables|CatchVars|GetMontageToPlay))
                    (Development|PrintString "Stop catch  montage")
                    (Ability|EndAbility))))
              (:Cancelled
                _)
              (:Completed
                _))
            (CallFunction|HandleMontageCompleted)
            (else
              ;[1 1 1] Check if we did not have the weapon out just play the GC and end
              (Development|PrintString "PlayNoCatchAnim")
              (Ability|ExecuteGameplayCueOnOwner (Variables|CatchVars|GetT_CatchGC))
              (Ability|EndAbility))))
        (:CastFailed
          _)))
    (:Is%20Not%20Valid
      _))
  (Ability|EndAbility))

(event Ability|EventOnEndAbility
  (params (bWasCancelled :type Boolean))
  (bind _playerpawn (Variables|CatchVars|GetPlayerPawn))
  ;[1 1 1] End the ability
  (Utilities|FlowControl|DoOnce false
    (:Completed
      ;[1 1 1] Unhide and reequip. If it is not the thrown weapon play the equip montage.
      (Utilities|IsValid _playerpawn
        (:Is%20Valid
          (Tags|UnBlockAbilitiesWithTags (Utilities|Casting|CastToFortAbilitySystemComponent (Ability|GetAbilitySystemComponent _playerpawn)) (Variables|CatchVars|GetAbilityBlockedTags) false)
          (Utilities|FlowControl|Sequence
            (:then_0
              (if (and (not (Parachute|IsSkydiving (Variables|CatchVars|GetPlayerPawn))) (Variables|Default|GetStartedMontage))
                (Weapon|SetHolsterWeapon (Variables|CatchVars|GetPlayerPawn) 0 false false (Variables|Default|GetHolsterId))))
            (:then_1
              (if (and (Parachute|IsSkydiving (Variables|CatchVars|GetPlayerPawn)) (Variables|Default|GetStartedMontage))
                (Weapon|SetHolsterWeapon (Variables|CatchVars|GetPlayerPawn) 0 false false (Variables|Default|GetHolsterId))))))
        (:Is%20Not%20Valid
          ;[1 1 1] Try to remove it from the controller if the pawn does not exist.
          (Tags|UnBlockAbilitiesWithTags (Utilities|Casting|CastToFortAbilitySystemComponent (Ability|GetAbilitySystemComponent (Fortnite|GetFortPlayerControllerfromActor _playerpawn))) (Variables|CatchVars|GetAbilityBlockedTags) false))) ; Make Sure it is still valid
    )))

(event Custom|HandleMontageCompleted
  ; Setup as an Event that we can override in the child Blueprints.
  (if (not (Networking|IsServer))
    (Utilities|FlowControl|DoOnce false
      (:Completed
        (Ability|EndAbility)))))
